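# Physics asset for the F_MED Silencer backpack cosmetic, defining secondary motion for its handle, straps, keychain, two swords and sword ribbon.
# Capsule bodies form chains around a box pack body; strap and sword roots join the pack with free rotation, while handle and keychain roots have no pack constraint.
# Swords barely swing, the ribbon hangs off sword 01 with limits widening toward its tip, and locked axes restrict most keychain links.
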PhysicsAsset /BRCosmetics/Accessories/FORT_Backpacks/Backpack_F_MED_Silencer/Meshes/F_MED_Silencer_Pack_Physics
bodies {
 dyn_backpack_handle_01 [{
   shapeType=Capsule
   length=3.6755
   radius=1.9723
   shapeName=dyn_backpack_handle_01_capsule
   center=<-0.9945 -0.3494 -0.1883>
   rotation=<rot 0.257 -90 0.3532>
  }]
 dyn_backpack_handle_root [{shapeType=Capsule, length=0, radius=0.5, shapeName=dyn_backpack_handle_root_capsule}]
 dyn_backpack_strap_left_01 [{
   shapeType=Capsule
   length=2.255
   radius=1.755
   shapeName=dyn_backpack_strap_left_01_capsule
   center=<2.2144 0.2123 0.2439>
   rotation=<rot 0 0 90>
  }]
 dyn_backpack_strap_left_02 [{
   shapeType=Capsule
   length=1.5481
   radius=1.9788
   shapeName=dyn_backpack_strap_left_02_capsule
   center=<2.2144 0.2123 0.2439>
   rotation=<rot 0 0 90>
  }]
 dyn_backpack_strap_left_root [{shapeType=Capsule, length=0.505, radius=0.505, shapeName=dyn_backpack_strap_left_root_capsule, rotation=<rot 0 0 90>}]
 dyn_backpack_strap_right_01 [{shapeType=Capsule, length=2.2514, radius=0.7619, shapeName=dyn_backpack_strap_right_01_capsule, center=<1 0 0>, rotation=<rot 0 0 90>}]
 dyn_backpack_strap_right_02 [{shapeType=Capsule, length=1.5425, radius=1.2974, shapeName=dyn_backpack_strap_right_02_capsule, center=<1 0 0>, rotation=<rot 0 0 90>}]
 dyn_backpack_strap_right_03 [{shapeType=Capsule, length=1.4898, radius=1.4666, shapeName=dyn_backpack_strap_right_03_capsule, center=<1 0 0>, rotation=<rot 0 0 90>}]
 dyn_backpack_strap_right_root [{shapeType=Capsule, length=0.505, radius=0.505, shapeName=dyn_backpack_strap_right_root_capsule, rotation=<rot 0 0 90>}]
 dyn_keychain_01 [{
   shapeType=Capsule
   length=0.75
   radius=1.2279
   shapeName=dyn_keychain_01_capsule
   center=<1.2311 0.0064 0>
   rotation=<rot 0.0008 -89.9998 90.001>
  }]
 dyn_keychain_02 [{
   shapeType=Capsule
   length=1.505
   radius=0.505
   shapeName=dyn_keychain_02_capsule
   center=<0.802 0.0057 0.0117>
   rotation=<rot 0 88.0667 -88.9318>
  }]
 dyn_keychain_03 [{
   shapeType=Capsule
   length=0
   radius=0.755
   shapeName=dyn_keychain_03_capsule
   center=<0.3768 -0.0013 0.0072>
   rotation=<rot 0 90.3988 -57.519>
  }]
 dyn_keychain_04 [{
   shapeType=Capsule
   length=2.25
   radius=1.5
   shapeName=dyn_keychain_04_capsule
   center=<2.7235 -0.1256 -1.0312>
   rotation=<rot 29.4532 -3.9698 -2.4702>
  }]
 dyn_keychain_root [{shapeType=Capsule, length=0.505, radius=0.505, shapeName=dyn_keychain_root_capsule, rotation=<rot 0 -90 0>}]
 dyn_sword_01 [{
   shapeType=Capsule
   length=39.599
   radius=1.7433
   shapeName=dyn_sword_01_capsule
   center=<1.3448 0.6178 -0.9875>
   rotation=<rot 50.3674 179.9988 180>
  }]
 dyn_sword_01_root [{shapeType=Capsule, length=0.505, radius=0.505, shapeName=dyn_sword_01_root_capsule, rotation=<rot 0 -90 0>}]
 dyn_sword_02 [{
   shapeType=Capsule
   length=45.4225
   radius=1.5
   shapeName=dyn_sword_02_capsule
   center=<4.0044 -0.112 2.5274>
   rotation=<rot -50.47 -0.0002 0>
  }]
 dyn_sword_02_root [{shapeType=Capsule, length=0.505, radius=0.505, shapeName=dyn_sword_02_root_capsule, rotation=<rot 0 -90 0>}]
 dyn_sword_ribbon_01 [{shapeType=Capsule, length=3.005, radius=0.505, shapeName=dyn_sword_ribbon_01_capsule, center=<2 0 0>, rotation=<rot 0 89.5654 -89.6374>}]
 dyn_sword_ribbon_02 [{shapeType=Capsule, length=3.6354, radius=0.5, shapeName=dyn_sword_ribbon_02_capsule, center=<1 0 0>, rotation=<rot 0 90.1008 -89.9753>}]
 dyn_sword_ribbon_03 [{shapeType=Capsule, length=2.505, radius=0.505, shapeName=dyn_sword_ribbon_03_capsule, center=<1 0 0>, rotation=<rot 0 -90.3644 -89.6527>}]
 pack [{
   shapeType=Box
   extentX=6
   extentY=11
   extentZ=11
   shapeName=pack_box
   center=<-9.1084 -10.5298 -12.9197>
   rotation=<rot -14.0625 0.0003 -0.0013>
  }]
}
constraints [
 {
  bone1Name=dyn_keychain_01
  bone2Name=dyn_keychain_root
  swing1LimitDegrees=26
  swing2LimitDegrees=26
  twistLimitDegrees=45
  twistMotion=Locked
 }
 {
  bone1Name=dyn_keychain_02
  bone2Name=dyn_keychain_01
  swing1LimitDegrees=45
  swing2LimitDegrees=45
  swing2Motion=Locked
  twistLimitDegrees=45
  twistMotion=Locked
 }
 {bone1Name=dyn_keychain_03, bone2Name=dyn_keychain_02, swing1LimitDegrees=45, swing2LimitDegrees=15, twistLimitDegrees=45, twistMotion=Locked}
 {bone1Name=dyn_keychain_04, bone2Name=dyn_keychain_03, swing1LimitDegrees=30, swing2LimitDegrees=30, twistLimitDegrees=5}
 {bone1Name=dyn_sword_ribbon_01, bone2Name=dyn_sword_01, swing1LimitDegrees=20, swing2LimitDegrees=20, twistLimitDegrees=10}
 {bone1Name=dyn_sword_ribbon_02, bone2Name=dyn_sword_ribbon_01, swing1LimitDegrees=30, swing2LimitDegrees=30, twistLimitDegrees=10}
 {bone1Name=dyn_sword_ribbon_03, bone2Name=dyn_sword_ribbon_02, swing1LimitDegrees=50, swing2LimitDegrees=50, twistLimitDegrees=10}
 {bone1Name=dyn_backpack_handle_01, bone2Name=dyn_backpack_handle_root, swing1LimitDegrees=10, swing2LimitDegrees=7.5, twistLimitDegrees=5}
 {
  bone1Name=dyn_backpack_strap_left_root
  bone2Name=pack
  swing1LimitDegrees=45
  swing1Motion=Free
  swing2LimitDegrees=45
  swing2Motion=Free
  twistLimitDegrees=45
  twistMotion=Free
 }
 {
  bone1Name=dyn_backpack_strap_left_01
  bone2Name=dyn_backpack_strap_left_root
  swing1LimitDegrees=30
  swing2LimitDegrees=30
  twistLimitDegrees=20
 }
 {
  bone1Name=dyn_backpack_strap_left_02
  bone2Name=dyn_backpack_strap_left_01
  swing1LimitDegrees=45
  swing2LimitDegrees=45
  twistLimitDegrees=5
 }
 {
  bone1Name=dyn_backpack_strap_right_root
  bone2Name=pack
  swing1LimitDegrees=45
  swing1Motion=Free
  swing2LimitDegrees=45
  swing2Motion=Free
  twistLimitDegrees=45
  twistMotion=Free
 }
 {
  bone1Name=dyn_backpack_strap_right_01
  bone2Name=dyn_backpack_strap_right_root
  swing1LimitDegrees=20
  swing2LimitDegrees=20
  twistLimitDegrees=7.5
 }
 {
  bone1Name=dyn_backpack_strap_right_02
  bone2Name=dyn_backpack_strap_right_01
  swing1LimitDegrees=28
  swing2LimitDegrees=28
  twistLimitDegrees=7.5
 }
 {
  bone1Name=dyn_backpack_strap_right_03
  bone2Name=dyn_backpack_strap_right_02
  swing1LimitDegrees=50
  swing2LimitDegrees=50
  twistLimitDegrees=7.5
 }
 {bone1Name=dyn_sword_02, bone2Name=dyn_sword_02_root, swing1LimitDegrees=2.5, swing2LimitDegrees=2.5, twistLimitDegrees=5}
 {
  bone1Name=dyn_sword_02_root
  bone2Name=pack
  swing1LimitDegrees=45
  swing1Motion=Free
  swing2LimitDegrees=45
  swing2Motion=Free
  twistLimitDegrees=45
  twistMotion=Free
 }
 {
  bone1Name=dyn_sword_01_root
  bone2Name=pack
  swing1LimitDegrees=45
  swing1Motion=Free
  swing2LimitDegrees=45
  swing2Motion=Free
  twistLimitDegrees=45
  twistMotion=Free
 }
 {bone1Name=dyn_sword_01, bone2Name=dyn_sword_01_root, swing1LimitDegrees=2.5, swing2LimitDegrees=5, twistLimitDegrees=2.5}
]
